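# Translucent UI material that draws a static dark background for the Freeform UI, with a diagonal gradient and angled shading.
# Scaled UVs pass through Rotators with constant time inputs, giving fixed angled axes that blend two color parameters, ColorA and ColorB.
# Two opposite-angled axes, split by anti-aliased thresholds, form hard-edged gradient wedges that are weighted, summed and added over the base through MultiplyAdd.
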Material /FortniteCoreUI/UI/Freeform/Art/M_UI_Freeform_Background
graph {
 properties {
  bAutomaticallySetUsageInEditor=false
  bUsedWithStaticMesh=false
  blendMode=BLEND_Translucent
  materialDomain=MD_UI
  usedShadingModels=Default Lit
 }
 nodes [
  n1 = MaterialFunctionCall() {materialFunction=&GetUserInterfaceUV}
  NamedRerouteDeclaration(None=@n1.Normalized UV) {name=Base UV, nodeColor=<rgba 0 0 0 1>}
  n2 = VectorParameter() {parameterName=ColorB, defaultValue=<rgba 0.007 0.008 0.0097 1>}
  n3 = VectorParameter() {parameterName=ColorA, defaultValue=<rgba 0.0203 0.0319 0.0452 1>}
  n4 = MaterialFunctionCall(UV=@n1.Normalized UV) {materialFunction=&MF_UI_UVScaling}
  n5 = Constant() {r=-0.2}
  n6 = ComponentMask(None=Saturate(None=Rotator(Coordinate=@n4.ScaleY, Time=@n5) {speed=6.2832})) {r=true}
  n7 = Constant() {r=0.2}
  n8 = Constant2Vector() {r=0.1}
  n9 = ComponentMask(None=Add(A=Rotator(Coordinate=@n4.ScaleY, Time=@n7) {speed=6.2832}, B=@n8)) {r=true}
  n10 = MaterialFunctionCall(DistanceField=@n9, Threshold=Constant() {r=0.5}) {materialFunction=&MF_UI_AntiAliasing}
  n11 = Constant() {r=-0.2}
  n12 = ComponentMask(None=Add(A=Rotator(Coordinate=@n4.ScaleY, Time=@n11) {speed=6.2832}, B=@n8)) {r=true}
  n13 = MaterialFunctionCall(DistanceField=@n12, Threshold=Constant() {r=0.5}) {materialFunction=&MF_UI_AntiAliasing}
  n14 = Saturate(None=@n12)
  n15 = Add(A=Multiply(A=Multiply(A=@n10.Inner) {constB=0.25}, B=@n9), B=Multiply(A=Multiply(A=@n13.Outer) {constB=0.3}, B=OneMinus(None=@n14)))
  Saturate(None=MaterialFunctionCall!(Base=LinearInterpolate(A=@n2.RGB, B=@n3.RGB, Alpha=@n6), Add=Saturate(None=@n15)) {materialFunction=/Engine/Functions/Engine_MaterialFunctions02/Math/MultiplyAdd}.Result)
 ]
}
